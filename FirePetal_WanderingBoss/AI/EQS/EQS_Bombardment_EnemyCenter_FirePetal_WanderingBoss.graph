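# EQS query for the FirePetal wandering boss that generates candidate bombardment center points around the querier, as the EnemyCenter name suggests.
# A spiral-pattern donut generator produces rings of points around the boss, projected with a navigation filter that excludes all water.
# A single 2D distance test only filters, rejecting points closer than a minimum distance to the locations from the bombardment PlayerTargets context.
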EnvQuery /FirePetal_WanderingBoss/AI/EQS/EQS_Bombardment_EnemyCenter_FirePetal_WanderingBoss
options [{
  generator {
   name=EnvQueryGenerator_Donut_0
   class=EnvQueryGenerator_Donut
   properties {
    bUseSpiralPattern=true
    center=&EnvQueryContext_Querier
    innerRadius=600
    numberOfRings=3
    outerRadius=1800
    pointsPerRing=8
    projectionData {
     extentX=0
     navigationFilter=&NavFilter_NPC_ExcludeAllWater_C
     postProjectionVerticalOffset=0
    }
   }
  }
  tests [EnvQueryTest_Distance_0 {
    class=EnvQueryTest_Distance
    properties {
     distanceTo=/FirePetal_WanderingBoss/AI/Context/EnvQueryContext_Bombardment_PlayerTargets_FirePetal_WanderingBoss.EnvQueryContext_Bombardment_PlayerTargets_FirePetal_WanderingBoss_C
     filterType=Minimum
     floatValueMax=242
     floatValueMin=250
     scoringEquation=Linear
     scoringFactor=1
     testComment=""
     testMode=Distance2D
     testPurpose=Filter
    }
   }]
 }]
