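# EQS query for the FirePetal wandering boss that generates candidate locations around island targets, apparently for its bombardment.
# A spiral donut generator places rings of points around the locations supplied by the boss's Bombardment IslandTargets context.
# Projection uses an NPC navigation filter that excludes all water.
# The query has no tests, so it applies no scoring or further filtering to the generated points.
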EnvQuery /FirePetal_WanderingBoss/AI/EQS/EQS_Bombardment_IslandCenter_FirePetal_WanderingBoss
options [{
  generator {
   name=EnvQueryGenerator_Donut_0
   class=EnvQueryGenerator_Donut
   properties {
    bUseSpiralPattern=true
    center=/FirePetal_WanderingBoss/AI/Context/EnvQueryContext_Bombardment_IslandTargets_FirePetal_WanderingBoss.EnvQueryContext_Bombardment_IslandTargets_FirePetal_WanderingBoss_C
    innerRadius=500
    numberOfRings=3
    outerRadius=1500
    pointsPerRing=7
    projectionData {
     extentX=0
     navigationFilter=&NavFilter_NPC_ExcludeAllWater_C
     postProjectionVerticalOffset=0
    }
   }
  }
  tests []
 }]
