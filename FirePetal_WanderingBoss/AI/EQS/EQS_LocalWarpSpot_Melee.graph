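# EQS query that picks a local melee warp spot around the current target for the FirePetal wandering boss.
# A navmesh-projected donut, sized by scalable-float bindings, is filtered by bot leash and minimum distance from the querier.
# Scoring strongly favors level height, plus line of sight, clear navmesh, target's front, safe dry ground and spacing from others.
# InverseLinear target-distance scoring favors closer spots, contradicting its "farther" comment.
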EnvQuery /FirePetal_WanderingBoss/AI/EQS/EQS_LocalWarpSpot_Melee
options [{
  generator {
   name=EnvQueryGenerator_Donut_0
   class=EnvQueryGenerator_Donut
   properties {
    bUseSpiralPattern=false
    center=&FortAthenaAIBot_EQSQueryContext_CurrentTarget
    innerRadius {
     dataBinding=&self:EnvQueryGenerator_Donut_0.FortAIDataProvider_ScalableFloat_3
     dataField=FloatValue
     defaultValue=300
    }
    numberOfRings {
     dataBinding=&self:EnvQueryGenerator_Donut_0.FortAIDataProvider_ScalableFloat_1
     dataField=IntValue
     defaultValue=3
    }
    outerRadius {
     dataBinding=&self:EnvQueryGenerator_Donut_0.FortAIDataProvider_ScalableFloat_4
     dataField=FloatValue
     defaultValue=1000
    }
    pointsPerRing {
     dataBinding=&self:EnvQueryGenerator_Donut_0.FortAIDataProvider_ScalableFloat_2
     dataField=IntValue
     defaultValue=8
    }
    projectionData {
     extentX=100
     navigationFilter=/Game/Athena/AI/NavFilters/NavFilter_NPC_ExcludeWater.NavFilter_NPC_ExcludeWater_C
     postProjectionVerticalOffset=30
    }
   }
  }
  tests [
   FortQueryTest_InsideAIBotLeash_0 {class=FortQueryTest_InsideAIBotLeash, properties {testPurpose=Filter}}
   EnvQueryTest_Trace_0 {
    class=EnvQueryTest_Trace
    properties {
     context=&FortAthenaAIBot_EQSQueryContext_CurrentTarget
     scoringFactor=2
     testComment=I prefer locations where I can see my target.
     testPurpose=Score
    }
   }
   EnvQueryTest_Distance_1 {
    class=EnvQueryTest_Distance
    properties {
     distanceTo=&FortAthenaAIBot_EQSQueryContext_CurrentTarget
     filterType=Range
     floatValueMin=0
     scoringEquation=InverseLinear
     scoringFactor=2
     testComment=I prefer locations farther from my target.
     testMode=Distance3D
     testPurpose=Score
    }
   }
   EnvQueryTest_Distance_0 {
    class=EnvQueryTest_Distance
    properties {
     distanceTo=&FortAthenaAIBot_EQSQueryContext_CurrentTarget
     filterType=Minimum
     floatValueMax=3
     floatValueMin=0
     scoringEquation=InverseLinear
     scoringFactor=10
     testComment=Prefer locations with a lesser height differential.
     testMode=DistanceAbsoluteZ
     testPurpose=Score
    }
   }
   EnvQueryTest_Distance_2 {
    class=EnvQueryTest_Distance
    properties {
     distanceTo=/Script/FortniteAIServer.FortQueryContext_AllEnemies
     filterType=Range
     floatValueMin=0
     scoringEquation=Linear
     scoringFactor=1
     testComment=I prefer to be farther from other hostiles.
     testMode=Distance3D
     testPurpose=Score
    }
   }
   EnvQueryTest_Distance_3 {
    class=EnvQueryTest_Distance
    properties {
     distanceTo=&FortQueryContext_NearbyAIPawns
     filterType=Range
     floatValueMin=0
     scoringEquation=Linear
     scoringFactor=0.5
     testComment=I prefer to stay away from other pawns.
     testMode=Distance3D
     testPurpose=Score
    }
   }
   EnvQueryTest_Distance_4 {
    class=EnvQueryTest_Distance
    properties {
     distanceTo=&FortQueryContext_NearbyAIPawnsMoveDestinations
     filterType=Range
     floatValueMin=0
     scoringEquation=Linear
     scoringFactor=0.5
     testComment=I prefer to avoid where other pawns are going.
     testMode=Distance3D
     testPurpose=Score
    }
   }
   EnvQueryTest_Project_0 {
    class=EnvQueryTest_Project
    properties {
     scoringFactor=4
     testComment=Give LOTS of score to things NOT on Water. We only use Water if we really need to.
     testPurpose=Score
     projectionData {
      navigationFilter=&NavFilter_NPC_ExcludeAllWater_C
      postProjectionVerticalOffset=30
      projectDown=1024
      projectUp=1024
      traceMode=Navigation
     }
    }
   }
   EnvQueryTest_Distance_5 {
    class=EnvQueryTest_Distance
    properties {
     distanceTo=&EnvQueryContext_Querier
     filterType=Minimum
     scoringEquation=Linear
     scoringFactor=1
     testComment=I prefer to be farther from self.
     testMode=Distance3D
     testPurpose=FilterAndScore
     floatValueMin {
      dataBinding=&self:EnvQueryTest_Distance_5.FortAIDataProvider_ScalableFloat_0
      dataField=FloatValue
      defaultValue=0
     }
    }
   }
   EnvQueryTest_Dot_0 {
    class=EnvQueryTest_Dot
    properties {
     clampMaxType=None
     clampMinType=None
     scoreClampMax=0
     scoreClampMin=0
     scoringFactor=3
     testComment=""
     lineA {rotation=&FortAthenaAIBot_EQSQueryContext_CurrentTarget}
     lineB {lineFrom=&FortAthenaAIBot_EQSQueryContext_CurrentTarget}
    }
   }
   EnvQueryTest_Project_1 {
    class=EnvQueryTest_Project
    properties {
     scoringFactor=4
     testComment=Give lots of score to things NOT on danger ground.
     testPurpose=Score
     projectionData {
      navigationFilter=/Game/Athena/AI/NavFilters/NavFilter_NPC_AnyTraversable_ExcludeDanger.NavFilter_NPC_AnyTraversable_ExcludeDanger_C
      postProjectionVerticalOffset=30
      projectDown=1024
      projectUp=1024
      traceMode=Navigation
     }
    }
   }
   FortQueryTest_NavmeshRaycast_0 {
    class=FortQueryTest_NavmeshRaycast
    properties {
     bDefineReferenceValue=false
     boolValue=true
     clampMaxType=None
     clampMinType=None
     filterOutDifferentFloors=false
     filterType=Minimum
     floatValueMax=1
     floatValueMin=0.2
     multipleContextFilterOp=AllPass
     multipleContextScoreOp=AverageScore
     navigationFilterOverride=/Game/Athena/AI/NavFilters/NavFilter_NPC_AllNavmesh.NavFilter_NPC_AllNavmesh_C
     normalizationType=Absolute
     raycastContext=&FortAthenaAIBot_EQSQueryContext_CurrentTarget
     referenceValue=0
     scoreClampMax=0
     scoreClampMin=0
     scoringEquation=Linear
     scoringFactor=4
     shouldRaycastToContext=true
     testComment=""
     testPurpose=Score
    }
   }
  ]
 }]
